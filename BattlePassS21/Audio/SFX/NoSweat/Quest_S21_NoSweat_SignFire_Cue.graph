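# Sound cue for a Season 21 Battle Pass NoSweat quest, playing a sign fire sound effect.
# A Random node picks one of three equally weighted wave variations.
# It routes through the non-critical in-game UI sound class, with no attenuation asset or other processing set on the cue.
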SoundCue /BattlePassS21/Audio/SFX/NoSweat/Quest_S21_NoSweat_SignFire_Cue
root=Random_0
nodes [
 {id=Random_0, children [WavePlayer_0,WavePlayer_1,WavePlayer_2], properties {weights [1,1,1]}}
 {id=WavePlayer_0, properties {soundWaveAssetPtr=&/Quest_S21_NoSweat_SignFire_03}}
 {id=WavePlayer_1, properties {soundWaveAssetPtr=&/Quest_S21_NoSweat_SignFire_01}}
 {id=WavePlayer_2, properties {soundWaveAssetPtr=&/Quest_S21_NoSweat_SignFire_02}}
]
properties {maxDistance=2097152, soundClassObject=&Fort_UI_InGame_NonCritical}
