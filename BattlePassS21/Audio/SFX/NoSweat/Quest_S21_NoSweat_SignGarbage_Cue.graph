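# Sound cue for the Season 21 battle pass No Sweat quest's "SignGarbage" sound effect.
# A random node picks between two wave variations with equal weight on each play.
# It routes through the non-critical in-game UI sound class at reduced volume, with a very large max distance.
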SoundCue /BattlePassS21/Audio/SFX/NoSweat/Quest_S21_NoSweat_SignGarbage_Cue
root=Random_0
nodes [
 {id=Random_0, children [WavePlayer_0,WavePlayer_1], properties {weights [1,1]}}
 {id=WavePlayer_0, properties {soundWaveAssetPtr=&/Quest_S21_NoSweat_SignGarbage_02}}
 {id=WavePlayer_1, properties {soundWaveAssetPtr=&/Quest_S21_NoSweat_SignGarbage_01}}
]
properties {maxDistance=2097152, soundClassObject=&Fort_UI_InGame_NonCritical, volumeMultiplier=0.6}
